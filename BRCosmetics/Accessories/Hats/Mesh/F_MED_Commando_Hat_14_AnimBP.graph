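# Post-attach Anim Blueprint for the female medium Commando Hat 14 cosmetic, adding secondary sway to its hair bones.
# The AnimGraph copies the pose from the attached parent mesh, then chains five AnimDynamics nodes on dyn_hair_01 through 05 before output.
# Each is a spring-driven, gravity-free single-bone sim in actor space with tight limits, blended at low alpha and gated by an LOD threshold.
# The event graph contains no logic.
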AnimBlueprint /BRCosmetics/Accessories/Hats/Mesh/F_MED_Commando_Hat_14_AnimBP
target_skeleton=&/F_MED_Commando_Hat_14_Skeleton
graphs [
 AnimGraph {nodes [
   n1 = Anim Dynamics - Bone: dyn_hair_01(ComponentPose=Local To Component(LocalPose=Copy Pose From Mesh() {bUseAttachedParent=true}), Alpha=0.126) {
    angularSpringConstant=20
    bAngularSpring=true
    bGravityOverrideInSimSpace=true
    bLinearSpring=true
    gravityScale=0
    lODThreshold=1
    linearSpringConstant=20
    simulationSpace=Actor
    alphaBoolBlend {blendOption=Linear}
    boundBone {boneName=dyn_hair_01}
    physicsBodyDefinitions [{
      sphereCollisionRadius=0
      boundBone {boneName=dyn_hair_01}
      boxExtents=<30 30 30>
      constraintSetup {
       twistAxis=AxisZ
       angularLimitsMax=<3 7.5 0>
       angularLimitsMin=<-3 -7.5 0>
       angularTarget=<0 0 0>
       linearAxesMax=<1.5 6 5>
       linearAxesMin=<-1.5 -6 -0.5>
      }
      localJointOffset=<0 0 0>
     }]
    relativeSpaceBone {boneName=dyn_hair_01}
   }
   n2 = Anim Dynamics - Bone: dyn_hair_02(ComponentPose=@n1, Alpha=0.2) {
    angularSpringConstant=65
    bAngularSpring=true
    bGravityOverrideInSimSpace=true
    bLinearSpring=true
    gravityScale=0
    lODThreshold=1
    linearSpringConstant=65
    simulationSpace=Actor
    alphaBoolBlend {blendOption=Linear}
    boundBone {boneName=dyn_hair_02}
    physicsBodyDefinitions [{
      sphereCollisionRadius=0
      boundBone {boneName=dyn_hair_02}
      boxExtents=<30 30 30>
      constraintSetup {
       twistAxis=AxisZ
       angularLimitsMax=<3 15 0>
       angularLimitsMin=<-3 -15 0>
       angularTarget=<0 0 0>
       linearAxesMax=<1.5 1 5>
       linearAxesMin=<-1.5 -1 -0.5>
      }
      localJointOffset=<0 0 0>
     }]
    relativeSpaceBone {boneName=dyn_hair_02}
   }
   n3 = Anim Dynamics - Bone: dyn_hair_03(ComponentPose=@n2, Alpha=0.23, ExternalForce="0,0,0") {
    angularSpringConstant=60
    bAngularSpring=true
    bGravityOverrideInSimSpace=true
    bLinearSpring=true
    gravityScale=0
    lODThreshold=1
    linearSpringConstant=60
    simulationSpace=Actor
    alphaBoolBlend {blendOption=Linear}
    boundBone {boneName=dyn_hair_03}
    physicsBodyDefinitions [{
      sphereCollisionRadius=0
      boundBone {boneName=dyn_hair_03}
      boxExtents=<30 30 30>
      constraintSetup {
       twistAxis=AxisZ
       angularLimitsMax=<3 7.5 5>
       angularLimitsMin=<-3 -7.5 -5>
       angularTarget=<0 0 0>
       linearAxesMax=<3 3 5>
       linearAxesMin=<-3 -3 -0.5>
      }
      localJointOffset=<0 0 0>
     }]
   }
   n4 = Anim Dynamics - Bone: dyn_hair_04(ComponentPose=@n3, Alpha=0.26, ExternalForce="0,0,0") {
    angularSpringConstant=80
    bAngularSpring=true
    bGravityOverrideInSimSpace=true
    bLinearSpring=true
    gravityScale=0
    lODThreshold=1
    linearSpringConstant=80
    simulationSpace=Actor
    alphaBoolBlend {blendOption=Linear}
    boundBone {boneName=dyn_hair_04}
    physicsBodyDefinitions [{
      sphereCollisionRadius=0
      boundBone {boneName=dyn_hair_04}
      boxExtents=<30 30 30>
      constraintSetup {
       twistAxis=AxisZ
       angularLimitsMax=<3 7.5 0>
       angularLimitsMin=<-3 -7.5 0>
       angularTarget=<0 0 0>
       linearAxesMax=<3 1.5 2.5>
       linearAxesMin=<-1.5 -0.5 -0.5>
      }
      localJointOffset=<0 0 0>
     }]
   }
   n5 = Anim Dynamics - Bone: dyn_hair_05(ComponentPose=@n4, Alpha=0.26, ExternalForce="0,0,0") {
    angularSpringConstant=80
    bAngularSpring=true
    bGravityOverrideInSimSpace=true
    bLinearSpring=true
    gravityScale=0
    lODThreshold=1
    linearSpringConstant=80
    simulationSpace=Actor
    alphaBoolBlend {blendOption=Linear}
    boundBone {boneName=dyn_hair_05}
    physicsBodyDefinitions [{
      sphereCollisionRadius=0
      boundBone {boneName=dyn_hair_05}
      boxExtents=<30 30 30>
      constraintSetup {
       twistAxis=AxisZ
       angularLimitsMax=<3 7.5 0>
       angularLimitsMin=<-3 -7.5 0>
       angularTarget=<0 0 0>
       linearAxesMax=<3 1.5 2.5>
       linearAxesMin=<-1.5 -0.5 -0.5>
      }
      localJointOffset=<0 0 0>
     }]
   }
   Output Pose(Result=Component To Local(ComponentPose=@n5)) {layerGroup=None}
   AnimGraphNode_SubInput_0 = Input Pose()
  ]}
 EventGraph {nodes [Event BlueprintUpdateAnimation()]}
]
variables []
