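# Animation Blueprint for the Dark Ninja hood cosmetic that adjusts its collar bones as the character moves.
# The EventGraph derives alphas from owner-mesh arm, spine and head deltas and pawn velocity, gated by targeting, melee and skydiving flags, with arm alphas zeroed during emotes.
# The AnimGraph copies the owner's pose, then chains additive bone-space Modify Bone nodes on five collar bones, mostly weighted by those alphas as hand-tuned correctives.
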AnimBlueprint /BRCosmetics/Accessories/Hats/Mesh/F_MED_DarkNinja_Hood_01_AnimBlueprint
target_skeleton=&/F_MED_DarkNinja_Hood_01_Skeleton
graphs [
 AnimGraph {nodes [
   n1 = Get Skydiving Movement to Right Alpha() -> (Skydiving Movement to Right Alpha:double)
   n2 = Get Melee Head turn to the left Alpha() -> (Melee Head turn to the left Alpha:double)
   n3 = Get Melee Head turn to the right Alpha() -> (Melee Head turn to the right Alpha:double)
   n4 = Get L Arm Raise Alpha() -> (L Arm Raise Alpha:float)
   n5 = Get Skeletal Mesh() -> (Skeletal Mesh:Skeletal Mesh Component Object Reference)
   n6 = Get R Arm Raise Alpha() -> (R Arm Raise Alpha:float)
   n7 = "Transform (Modify) Bone - Bone: deform_collar_front_r_01"(
    ComponentPose=Local To Component(LocalPose=Copy Pose From Mesh(SourceMeshComponent=@n5))
    Alpha=@n6
    Translation="-2.751,-0.615,-7.286"
    Rotation="-0.017,0,-13.063"
    Scale="1,1,1"
   ) {
    rotationMode=BMM_Additive
    rotationSpace=BCS_BoneSpace
    translationMode=BMM_Additive
    translationSpace=BCS_BoneSpace
    alphaBoolBlend {blendOption=Linear}
    boneToModify {boneName=deform_collar_front_r_01}
    rotation {pitch=-0.017, roll=-13.063}
    translation=<-2.751 -0.615 -7.286>
   }
   n8 = "Transform (Modify) Bone - Bone: deform_collar_back_r_01"(ComponentPose=@n7, Alpha=@n6, Translation="-1.468,5.336,-0.679", Rotation="-5.180,-29.504,-20.744", Scale="1,1,1") {
    rotationMode=BMM_Additive
    rotationSpace=BCS_BoneSpace
    translationMode=BMM_Additive
    translationSpace=BCS_BoneSpace
    alphaBoolBlend {blendOption=Linear}
    boneToModify {boneName=deform_collar_back_r_01}
    rotation=<rot -5.1799 -29.504 -20.744>
    translation=<-1.468 5.336 -0.679>
   }
   n9 = Get Spine rotation to the Left while Targeting Alpha() -> (Spine rotation to the Left while Targeting Alpha:double)
   n10 = "Transform (Modify) Bone - Bone: deform_collar_front_r_01"(ComponentPose=@n8, Alpha=@n9, Translation="-1.516,2.274,-1.840", Rotation="-7.772,23.337,-17.933", Scale="1,1,1") {
    rotationMode=BMM_Additive
    rotationSpace=BCS_BoneSpace
    translationMode=BMM_Additive
    translationSpace=BCS_BoneSpace
    alphaBoolBlend {blendOption=Linear}
    boneToModify {boneName=deform_collar_front_r_01}
    rotation=<rot -7.772 23.337 -17.9329>
    translation=<-1.516 2.274 -1.84>
   }
   n11 = "Transform (Modify) Bone - Bone: deform_collar_back_r_01"(ComponentPose=@n10, Alpha=@n9, Translation="1.225,8.685,4.584", Rotation="0,0,0", Scale="1,1,1") {
    rotationMode=BMM_Additive
    rotationSpace=BCS_BoneSpace
    translationMode=BMM_Additive
    translationSpace=BCS_BoneSpace
    alphaBoolBlend {blendOption=Linear}
    boneToModify {boneName=deform_collar_back_r_01}
    translation=<1.225 8.685 4.584>
   }
   n12 = "Transform (Modify) Bone - Bone: deform_collar_back_l_01"(ComponentPose=@n11, Alpha=@n9, Translation="0.888,0.696,-2.269", Rotation="0,0,0", Scale="1,1,1") {
    rotationMode=BMM_Additive
    rotationSpace=BCS_BoneSpace
    translationMode=BMM_Additive
    translationSpace=BCS_BoneSpace
    alphaBoolBlend {blendOption=Linear}
    boneToModify {boneName=deform_collar_back_l_01}
    translation=<0.888 0.696 -2.269>
   }
   n13 = "Transform (Modify) Bone - Bone: deform_collar_front_l_01"(ComponentPose=@n12, Alpha=@n9, Translation="0,0,0", Rotation="0,0,0", Scale="1,1,1") {
    rotationMode=BMM_Additive
    rotationSpace=BCS_BoneSpace
    translationMode=BMM_Additive
    translationSpace=BCS_BoneSpace
    alphaBoolBlend {blendOption=Linear}
    boneToModify {boneName=deform_collar_front_l_01}
   }
   n14 = Get Head turn to the left Alpha() -> (Head turn to the left Alpha:double)
   n15 = "Transform (Modify) Bone - Bone: deform_collar_back_r_01"(ComponentPose=@n13, Alpha=@n14, Translation="0.003,-0.077,-1.321", Rotation="0,0,0", Scale="1,1,1") {
    rotationMode=BMM_Additive
    rotationSpace=BCS_BoneSpace
    translationMode=BMM_Additive
    translationSpace=BCS_BoneSpace
    alphaBoolBlend {blendOption=Linear}
    boneToModify {boneName=deform_collar_back_r_01}
    translation=<0.003 -0.077 -1.321>
   }
   n16 = "Transform (Modify) Bone - Bone: deform_collar_front_r_01"(ComponentPose=@n15, Alpha=@n14, Translation="-2.767,0.391,-2.878", Rotation="5.000,-0.002,0.002", Scale="1,1,1") {
    rotationMode=BMM_Additive
    rotationSpace=BCS_BoneSpace
    translationMode=BMM_Additive
    translationSpace=BCS_BoneSpace
    alphaBoolBlend {blendOption=Linear}
    boneToModify {boneName=deform_collar_front_r_01}
    rotation=<rot 5 -0.002 0.002>
    translation=<-2.767 0.391 -2.878>
   }
   n17 = "Transform (Modify) Bone - Bone: deform_collar_front_l_01"(ComponentPose=@n16, Alpha=@n4, Translation="-0.530,1.221,-1.170", Rotation="-0.000,-0.107,0.072", Scale="1,1,1") {
    rotationMode=BMM_Additive
    rotationSpace=BCS_BoneSpace
    translationMode=BMM_Additive
    translationSpace=BCS_BoneSpace
    alphaBoolBlend {blendOption=Linear}
    boneToModify {boneName=deform_collar_front_l_01}
    rotation=<rot -0.0001 -0.1069 0.072>
    translation=<-0.53 1.221 -1.17>
   }
   n18 = "Transform (Modify) Bone - Bone: deform_collar_back_l_01"(ComponentPose=@n17, Alpha=@n4, Translation="0.000,-0.013,-2.539", Rotation="-0.000,-0.009,0.002", Scale="1,1,1") {
    rotationMode=BMM_Additive
    rotationSpace=BCS_BoneSpace
    translationMode=BMM_Additive
    translationSpace=BCS_BoneSpace
    alphaBoolBlend {blendOption=Linear}
    boneToModify {boneName=deform_collar_back_l_01}
    rotation=<rot 0 -0.009 0.002>
    translation {y=-0.013, z=-2.539}
   }
   n19 = Get Head look Up Alpha() -> (Head look Up Alpha:double)
   n20 = "Transform (Modify) Bone - Bone: deform_collar_back_01"(ComponentPose=@n18, Alpha=@n19, Translation="0,0,-2", Rotation="0,0,0", Scale="1,1,1") {
    rotationMode=BMM_Additive
    rotationSpace=BCS_BoneSpace
    translationMode=BMM_Additive
    translationSpace=BCS_BoneSpace
    alphaBoolBlend {blendOption=Linear}
    boneToModify {boneName=deform_collar_back_01}
    translation {z=-2}
   }
   n21 = "Transform (Modify) Bone - Bone: deform_collar_back_r_01"(ComponentPose=@n20, Alpha=@n19, Translation="1,0,-2", Rotation="0,0,0", Scale="1,1,1") {
    rotationMode=BMM_Additive
    rotationSpace=BCS_BoneSpace
    translationMode=BMM_Additive
    translationSpace=BCS_BoneSpace
    alphaBoolBlend {blendOption=Linear}
    boneToModify {boneName=deform_collar_back_r_01}
    translation {x=1, z=-2}
   }
   n22 = "Transform (Modify) Bone - Bone: deform_collar_back_l_01"(ComponentPose=@n21, Alpha=@n19, Translation="0.000,-0.000,-2.664", Rotation="0.001,-0.020,0.005", Scale="1,1,1") {
    rotationMode=BMM_Additive
    rotationSpace=BCS_BoneSpace
    translationMode=BMM_Additive
    translationSpace=BCS_BoneSpace
    alphaBoolBlend {blendOption=Linear}
    boneToModify {boneName=deform_collar_back_l_01}
    rotation=<rot 0.001 -0.0199 0.005>
    translation {z=-2.664}
   }
   n23 = "Transform (Modify) Bone - Bone: deform_collar_front_r_01"(ComponentPose=@n22, Alpha=@n19, Translation="-0.505,-0.015,-0.638", Rotation="-10.249,0.272,0.149", Scale="1,1,1") {
    rotationMode=BMM_Additive
    rotationSpace=BCS_BoneSpace
    translationMode=BMM_Additive
    translationSpace=BCS_BoneSpace
    alphaBoolBlend {blendOption=Linear}
    boneToModify {boneName=deform_collar_front_r_01}
    rotation=<rot -10.249 0.272 0.149>
    translation=<-0.505 -0.015 -0.638>
   }
   n24 = "Transform (Modify) Bone - Bone: deform_collar_front_l_01"(ComponentPose=@n23, Alpha=@n19, Translation="0.239,-0.000,-4.340", Rotation="-10.468,-0.312,-0.073", Scale="1,1,1") {
    rotationMode=BMM_Additive
    rotationSpace=BCS_BoneSpace
    translationMode=BMM_Additive
    translationSpace=BCS_BoneSpace
    alphaBoolBlend {blendOption=Linear}
    boneToModify {boneName=deform_collar_front_l_01}
    rotation=<rot -10.468 -0.3119 -0.073>
    translation {x=0.239, z=-4.34}
   }
   n25 = Get Head Tilt To Right Alpha() -> (Head Tilt To Right Alpha:double)
   n26 = "Transform (Modify) Bone - Bone: deform_collar_back_r_01"(ComponentPose=@n24, Alpha=@n25, Translation="-0.024,1.130,4.981", Rotation="40,0,0", Scale="1,1,1") {
    rotationMode=BMM_Additive
    rotationSpace=BCS_BoneSpace
    translationMode=BMM_Additive
    translationSpace=BCS_BoneSpace
    alphaBoolBlend {blendOption=Linear}
    boneToModify {boneName=deform_collar_back_r_01}
    rotation {pitch=40}
    translation=<-0.024 1.13 4.981>
   }
   n27 = "Transform (Modify) Bone - Bone: deform_collar_back_01"(ComponentPose=@n26, Alpha=@n25, Translation="0,0,0", Rotation="0,0,0", Scale="1,1,1") {
    rotationMode=BMM_Additive
    rotationSpace=BCS_BoneSpace
    translationMode=BMM_Additive
    translationSpace=BCS_BoneSpace
    alphaBoolBlend {blendOption=Linear}
    boneToModify {boneName=deform_collar_back_01}
   }
   n28 = "Transform (Modify) Bone - Bone: deform_collar_front_r_01"(ComponentPose=@n27, Alpha=@n25, Translation="0,0,0", Rotation="0,0,0", Scale="1,1,1") {
    rotationMode=BMM_Additive
    rotationSpace=BCS_BoneSpace
    translationMode=BMM_Additive
    translationSpace=BCS_BoneSpace
    alphaBoolBlend {blendOption=Linear}
    boneToModify {boneName=deform_collar_front_r_01}
   }
   n29 = "Transform (Modify) Bone - Bone: deform_collar_front_l_01"(ComponentPose=@n28, Alpha=@n25, Translation="-2.254,6.249,-4.786", Rotation="0,0,0", Scale="1,1,1") {
    rotationMode=BMM_Additive
    rotationSpace=BCS_BoneSpace
    translationMode=BMM_Additive
    translationSpace=BCS_BoneSpace
    alphaBoolBlend {blendOption=Linear}
    boneToModify {boneName=deform_collar_front_l_01}
    translation=<-2.254 6.249 -4.786>
   }
   n30 = "Transform (Modify) Bone - Bone: deform_collar_back_01"(ComponentPose=@n29, Alpha=@n3, Translation="0.000,0.000,1.921", Rotation="0,0,0", Scale="1,1,1") {
    rotationMode=BMM_Additive
    rotationSpace=BCS_BoneSpace
    translationMode=BMM_Additive
    translationSpace=BCS_BoneSpace
    alphaBoolBlend {blendOption=Linear}
    alphaScaleBiasClamp {interpSpeedDecreasing=1, interpSpeedIncreasing=30}
    boneToModify {boneName=deform_collar_back_01}
    translation {z=1.921}
   }
   n31 = "Transform (Modify) Bone - Bone: deform_collar_back_r_01"(ComponentPose=@n30, Alpha=@n3, Translation="0.000,0.000,1.977", Rotation="0,0,0", Scale="1,1,1") {
    rotationMode=BMM_Additive
    rotationSpace=BCS_BoneSpace
    translationMode=BMM_Additive
    translationSpace=BCS_BoneSpace
    alphaBoolBlend {blendOption=Linear}
    alphaScaleBiasClamp {interpSpeedDecreasing=1, interpSpeedIncreasing=30}
    boneToModify {boneName=deform_collar_back_r_01}
    translation {z=1.977}
   }
   n32 = "Transform (Modify) Bone - Bone: deform_collar_back_l_01"(ComponentPose=@n31, Alpha=@n3, Translation="0,0,0", Rotation="0,0,0", Scale="1,1,1") {
    rotationMode=BMM_Additive
    rotationSpace=BCS_BoneSpace
    translationMode=BMM_Additive
    translationSpace=BCS_BoneSpace
    alphaBoolBlend {blendOption=Linear}
    alphaScaleBiasClamp {interpSpeedDecreasing=1, interpSpeedIncreasing=30}
    boneToModify {boneName=deform_collar_back_l_01}
   }
   n33 = "Transform (Modify) Bone - Bone: deform_collar_front_r_01"(ComponentPose=@n32, Alpha=@n3, Translation="-2.974,-2.728,0.816", Rotation="51.324,-2.424,-1.089", Scale="1,1,1") {
    rotationMode=BMM_Additive
    rotationSpace=BCS_BoneSpace
    translationMode=BMM_Additive
    translationSpace=BCS_BoneSpace
    alphaBoolBlend {blendOption=Linear}
    alphaScaleBiasClamp {interpSpeedDecreasing=1, interpSpeedIncreasing=1}
    boneToModify {boneName=deform_collar_front_r_01}
    rotation=<rot 51.324 -2.4241 -1.089>
    translation=<-2.974 -2.728 0.816>
   }
   n34 = "Transform (Modify) Bone - Bone: deform_collar_front_l_01"(ComponentPose=@n33, Alpha=@n3, Translation="0,0,0", Rotation="0,0,0", Scale="1,1,1") {
    rotationMode=BMM_Additive
    rotationSpace=BCS_BoneSpace
    translationMode=BMM_Additive
    translationSpace=BCS_BoneSpace
    alphaBoolBlend {blendOption=Linear}
    alphaScaleBiasClamp {interpSpeedDecreasing=1, interpSpeedIncreasing=30}
    boneToModify {boneName=deform_collar_front_l_01}
   }
   n35 = "Transform (Modify) Bone - Bone: deform_collar_back_01"(ComponentPose=@n34, Alpha=@n2, Translation="-0.000,14.421,-1.212", Rotation="0,0,0", Scale="1,1,1") {
    rotationMode=BMM_Additive
    rotationSpace=BCS_BoneSpace
    translationMode=BMM_Additive
    translationSpace=BCS_BoneSpace
    alphaBoolBlend {blendOption=Linear}
    alphaScaleBiasClamp {interpSpeedDecreasing=1, interpSpeedIncreasing=30}
    boneToModify {boneName=deform_collar_back_01}
    translation {y=14.421, z=-1.212}
   }
   n36 = "Transform (Modify) Bone - Bone: deform_collar_back_r_01"(ComponentPose=@n35, Alpha=@n2, Translation="0.000,7.108,-0.000", Rotation="0,0,0", Scale="1,1,1") {
    rotationMode=BMM_Additive
    rotationSpace=BCS_BoneSpace
    translationMode=BMM_Additive
    translationSpace=BCS_BoneSpace
    alphaBoolBlend {blendOption=Linear}
    alphaScaleBiasClamp {interpSpeedDecreasing=1, interpSpeedIncreasing=30}
    boneToModify {boneName=deform_collar_back_r_01}
    translation {y=7.108}
   }
   n37 = "Transform (Modify) Bone - Bone: deform_collar_back_l_01"(ComponentPose=@n36, Alpha=@n2, Translation="0.000,3.203,2.024", Rotation="0,0,0", Scale="1,1,1") {
    rotationMode=BMM_Additive
    rotationSpace=BCS_BoneSpace
    translationMode=BMM_Additive
    translationSpace=BCS_BoneSpace
    alphaBoolBlend {blendOption=Linear}
    alphaScaleBiasClamp {interpSpeedDecreasing=1, interpSpeedIncreasing=30}
    boneToModify {boneName=deform_collar_back_l_01}
    translation {y=3.203, z=2.024}
   }
   n38 = "Transform (Modify) Bone - Bone: deform_collar_front_r_01"(ComponentPose=@n37, Alpha=@n2, Translation="0,0,0", Rotation="0,0,0", Scale="1,1,1") {
    rotationMode=BMM_Additive
    rotationSpace=BCS_BoneSpace
    translationMode=BMM_Additive
    translationSpace=BCS_BoneSpace
    alphaBoolBlend {blendOption=Linear}
    alphaScaleBiasClamp {interpSpeedDecreasing=1, interpSpeedIncreasing=30}
    boneToModify {boneName=deform_collar_front_r_01}
   }
   n39 = "Transform (Modify) Bone - Bone: deform_collar_front_l_01"(ComponentPose=@n38, Alpha=@n2, Translation="-1.079,5.280,-1.291", Rotation="57.825,-64.932,-66.736", Scale="1,1,1") {
    rotationMode=BMM_Additive
    rotationSpace=BCS_BoneSpace
    translationMode=BMM_Additive
    translationSpace=BCS_BoneSpace
    alphaBoolBlend {blendOption=Linear}
    alphaScaleBiasClamp {bInterpResult=true, interpSpeedDecreasing=20, interpSpeedIncreasing=20}
    boneToModify {boneName=deform_collar_front_l_01}
    rotation=<rot 57.825 -64.9319 -66.7359>
    translation=<-1.079 5.28 -1.291>
   }
   n40 = "Transform (Modify) Bone - Bone: deform_collar_back_l_01"!(ComponentPose=@n39, bAlphaBoolEnabled:bool=True, Translation="5.454,-0.000,-0.000", Rotation="0,0,0", Scale="1,1,1") -> (Pose) {
    alphaInputType=Bool
    rotationMode=BMM_Additive
    rotationSpace=BCS_BoneSpace
    translationMode=BMM_Additive
    translationSpace=BCS_BoneSpace
    alphaBoolBlend {blendInTime=0.3, blendOption=ExpInOut, blendOutTime=0.3}
    alphaScaleBiasClamp {bInterpResult=true}
    boneToModify {boneName=deform_collar_back_l_01}
    translation {x=5.454}
   }
   n41 = "Transform (Modify) Bone - Bone: deform_collar_back_01"!(ComponentPose=@n40, bAlphaBoolEnabled:bool=True, Translation="4.177,-0.000,-0.000", Rotation="0,0,0", Scale="1,1,1") -> (Pose) {
    alphaInputType=Bool
    rotationMode=BMM_Additive
    rotationSpace=BCS_BoneSpace
    translationMode=BMM_Additive
    translationSpace=BCS_BoneSpace
    alphaBoolBlend {blendInTime=0.3, blendOption=ExpInOut, blendOutTime=0.3}
    alphaScaleBiasClamp {bInterpResult=true}
    boneToModify {boneName=deform_collar_back_01}
    translation {x=4.177}
   }
   n42 = "Transform (Modify) Bone - Bone: deform_collar_back_r_01"!(ComponentPose=@n41, bAlphaBoolEnabled:bool=True, Translation="6.734,0.000,0.000", Rotation="0,0,0", Scale="1,1,1") -> (Pose) {
    alphaInputType=Bool
    rotationMode=BMM_Additive
    rotationSpace=BCS_BoneSpace
    translationMode=BMM_Additive
    translationSpace=BCS_BoneSpace
    alphaBoolBlend {blendInTime=0.3, blendOption=ExpInOut, blendOutTime=0.3}
    alphaScaleBiasClamp {bInterpResult=true}
    boneToModify {boneName=deform_collar_back_r_01}
    translation {x=6.734}
   }
   n43 = "Transform (Modify) Bone - Bone: deform_collar_front_r_01"!(ComponentPose=@n42, bAlphaBoolEnabled:bool=True, Translation="0,0,0", Rotation="0,0,0", Scale="1,1,1") -> (Pose) {
    alphaInputType=Bool
    rotationMode=BMM_Additive
    rotationSpace=BCS_BoneSpace
    translationMode=BMM_Additive
    translationSpace=BCS_BoneSpace
    alphaBoolBlend {blendInTime=0.3, blendOption=ExpInOut, blendOutTime=0.3}
    alphaScaleBiasClamp {bInterpResult=true}
    boneToModify {boneName=deform_collar_front_r_01}
   }
   n44 = "Transform (Modify) Bone - Bone: deform_collar_front_l_01"!(ComponentPose=@n43, bAlphaBoolEnabled:bool=True, Translation="-0.367,0.344,-1.893", Rotation="0,0,0", Scale="1,1,1") -> (Pose) {
    alphaInputType=Bool
    rotationMode=BMM_Additive
    rotationSpace=BCS_BoneSpace
    translationMode=BMM_Additive
    translationSpace=BCS_BoneSpace
    alphaBoolBlend {blendInTime=0.3, blendOption=ExpInOut, blendOutTime=0.3}
    alphaScaleBiasClamp {bInterpResult=true}
    boneToModify {boneName=deform_collar_front_l_01}
    translation=<-0.367 0.344 -1.893>
   }
   n45 = "Transform (Modify) Bone - Bone: deform_collar_front_l_01"!(ComponentPose=@n44, Alpha:float=@n1, Translation="-1.103,-1.040,-2.916", Rotation="0,0,0", Scale="1,1,1") -> (Pose) {
    rotationMode=BMM_Additive
    rotationSpace=BCS_BoneSpace
    translationMode=BMM_Additive
    translationSpace=BCS_BoneSpace
    alphaBoolBlend {blendInTime=0.3, blendOption=ExpInOut, blendOutTime=0.3}
    boneToModify {boneName=deform_collar_front_l_01}
    translation=<-1.103 -1.04 -2.916>
   }
   n46 = "Transform (Modify) Bone - Bone: deform_collar_front_r_01"!(ComponentPose=@n45, Alpha:float=@n1, Translation="3.012,0.207,1.425", Rotation="0,0,0", Scale="1,1,1") -> (Pose) {
    rotationMode=BMM_Additive
    rotationSpace=BCS_BoneSpace
    translationMode=BMM_Additive
    translationSpace=BCS_BoneSpace
    alphaBoolBlend {blendInTime=0.3, blendOption=ExpInOut, blendOutTime=0.3}
    boneToModify {boneName=deform_collar_front_r_01}
    translation=<3.012 0.207 1.425>
   }
   n47 = Get Skydiving Movement to Left Alpha() -> (Skydiving Movement to Left Alpha:double)
   n48 = "Transform (Modify) Bone - Bone: deform_collar_front_l_01"(ComponentPose=@n46, Alpha=@n47, Translation="0.387,0.519,0.057", Rotation="0,0,0", Scale="1,1,1") {
    rotationMode=BMM_Additive
    rotationSpace=BCS_BoneSpace
    translationMode=BMM_Additive
    translationSpace=BCS_BoneSpace
    alphaBoolBlend {blendInTime=0.3, blendOption=ExpInOut, blendOutTime=0.3}
    boneToModify {boneName=deform_collar_front_l_01}
    translation=<0.387 0.519 0.057>
   }
   n49 = "Transform (Modify) Bone - Bone: deform_collar_front_r_01"(ComponentPose=@n48, Alpha=@n47, Translation="1.206,-0.052,-0.359", Rotation="0,0,0", Scale="1,1,1") {
    rotationMode=BMM_Additive
    rotationSpace=BCS_BoneSpace
    translationMode=BMM_Additive
    translationSpace=BCS_BoneSpace
    alphaBoolBlend {blendInTime=0.3, blendOption=ExpInOut, blendOutTime=0.3}
    boneToModify {boneName=deform_collar_front_r_01}
    translation=<1.206 -0.052 -0.359>
   }
   Output Pose(Result=Component To Local(ComponentPose=@n49)) {layerGroup=None}
  ]}
 EventGraph {nodes [
   Event BlueprintUpdateAnimation()
   > n1 = Sequence!(execute) -> (then_0, then_1, then_2:Exec, then_3:Exec)
   n2 = Branch(execute=@n1.then_1, Condition=Get bIsPlayingEmote())
   > Set R Arm Raise Alpha(R Arm Raise Alpha:float=0) -> (Output_Get:float)
   > Set L Arm Raise Alpha(L Arm Raise Alpha:float=0) -> (Output_Get:float)
   Set R Arm Raise Alpha(execute=@n2.else, R Arm Raise Alpha=Get R Arm Raise Alpha())
   > Set L Arm Raise Alpha(L Arm Raise Alpha=Get L Arm Raise Alpha())
   n3 = TryGetPawnOwner()
   n4 = GetVelocity!(self:Actor Object Reference=@n3) -> (ReturnValue:Vector)
   n5 = Dot Product!(A:Vector=GetActorRightVector!(self:Actor Object Reference=@n3) -> (ReturnValue:Vector), B:Vector=@n4) -> (ReturnValue:double)
   n6 = Set Movement to Right Alpha(execute=@n1.then_2, Movement to Right Alpha:double=MapRangeClamped(Value=@n5, InRangeA=0, InRangeB=230)) -> (Output_Get:double)
   > Set Skydiving Movement to Right Alpha(Skydiving Movement to Right Alpha:double=float * float(A=@n6, B:bool=Get bIsSkyDiving()) {is=title K2Node_PromotableOperator}) -> (Output_Get:double)
   > n7 = Set Movement to Left Alpha(Movement to Left Alpha:double=MapRangeClamped(Value=@n5, InRangeA=0, InRangeB=-230)) -> (Output_Get:double)
   > Set Skydiving Movement to Left Alpha(Skydiving Movement to Left Alpha:double=float * float(A=@n7, B=Get bIsSkyDiving())) -> (Output_Get:double)
   n8 = Get Skeletal Mesh()
   Set R Arm Raise Alpha(
    execute=@n1.then_0
    R Arm Raise Alpha=float + float(
     A=GetAlphaFromReferenceTransformDelta(SkeletalMeshComponent=@n8, BoneName=upperarm_r, BaseName=spine_05, InputRange=(X=60,Y=105), TwistAxis="0,0,1")
     B=GetAlphaFromReferenceTransformDelta(SkeletalMeshComponent=@n8, BoneName=upperarm_r, BaseName=spine_05, InputRange=(X=-180,Y=-60), OutputRange=(X=1,Y=0), TwistAxis="0,0,1")
    )
   )
   > Set L Arm Raise Alpha(L Arm Raise Alpha=float + float(
     A=GetAlphaFromReferenceTransformDelta(SkeletalMeshComponent=@n8, BoneName=upperarm_l, BaseName=spine_05, InputRange=(X=60,Y=105), TwistAxis="0,0,1")
     B=GetAlphaFromReferenceTransformDelta(SkeletalMeshComponent=@n8, BoneName=upperarm_l, BaseName=spine_05, InputRange=(X=-180,Y=-60), OutputRange=(X=1,Y=0), TwistAxis="0,0,1")
    ))
   > Set Spine rotation to the Left while Targeting Alpha(Spine rotation to the Left while Targeting Alpha:double=float * float(
     A=GetAlphaFromReferenceTransformDelta(SkeletalMeshComponent=@n8, BoneName=spine_05, BaseName=pelvis, DeltaComponentType=EulerX, InputRange=(X=-10,Y=7), TwistAxis="1,0,0")
     B=Get bIsTargeting()
    )) -> (Output_Get:double)
   > Set Spine rotation to the Right while Targeting Alpha(Spine rotation to the Right while Targeting Alpha:double=float * float(
     A=GetAlphaFromReferenceTransformDelta(SkeletalMeshComponent=@n8, BoneName=spine_05, BaseName=pelvis, DeltaComponentType=EulerX, InputRange=(X=-10,Y=-45), TwistAxis="0,0,1")
     B=Get bIsTargeting()
    )) -> (Output_Get:double)
   n9 = GetAlphaFromReferenceTransformDelta(SkeletalMeshComponent=@n8, BoneName=head, BaseName=spine_05, DeltaComponentType=EulerX, InputRange=(X=0,Y=-45), TwistAxis="0,0,1")
   > Set Melee Head turn to the right Alpha(Melee Head turn to the right Alpha:double=float * float(A:bool=Get bIsPlayingMeleeAnim(), B:double=@n9)) -> (Output_Get:double)
   > Set Head turn to the right Alpha(Head turn to the right Alpha:double=@n9) -> (Output_Get:double)
   n10 = GetAlphaFromReferenceTransformDelta(SkeletalMeshComponent=@n8, BoneName=head, BaseName=spine_05, DeltaComponentType=EulerX, InputRange=(X=10,Y=45), TwistAxis="0,0,1")
   > Set Melee Head turn to the left Alpha(Melee Head turn to the left Alpha:double=float * float(A=Get bIsPlayingMeleeAnim(), B=@n10)) -> (Output_Get:double)
   > Set Head turn to the left Alpha(Head turn to the left Alpha:double=@n10) -> (Output_Get:double)
   > Set Head look Up Alpha(Head look Up Alpha:double=GetAlphaFromReferenceTransformDelta(SkeletalMeshComponent=@n8, BoneName=head, BaseName=spine_05, InputRange=(X=0,Y=-45), TwistAxis="0,0,1")) -> (Output_Get:double)
   > Set Head Tilt To Right Alpha(Head Tilt To Right Alpha:double=GetAlphaFromReferenceTransformDelta(SkeletalMeshComponent=@n8, BoneName=head, BaseName=spine_05, DeltaComponentType=EulerY, InputRange=(X=0,Y=30), TwistAxis="0,0,1")) -> (Output_Get:double)
   > Set Is Skydiving Floating or Dropping(Is Skydiving Floating or Dropping:bool=OR Boolean(A=Get bIsSkydiveFloating(), B=Get bIsSkydiveDiveMode())) -> (Output_Get:bool)
   n11 = Dot Product(A=@n4, B=GetActorForwardVector!(self:Actor Object Reference=@n3) -> (ReturnValue:Vector))
   n12 = Set Movement Forward or Backward Alpha(execute=@n1.then_3, Movement Forward or Backward Alpha:double=MapRangeClamped(Value="Absolute (Float)"(A=@n11), InRangeA=0, InRangeB=300)) -> (Output_Get:double)
   > Set Skydiving Movement Forward or Backward Alpha(Skydiving Movement Forward or Backward Alpha:double=float * float(A:double=@n12, B:bool=Get bIsSkyDiving())) -> (Output_Get:double)
   Event BlueprintInitializeAnimation()
   > Set Skeletal Mesh(Skeletal Mesh:Skeletal Mesh Component Object Reference=GetOwnerSkeletalMeshComponent()) -> (Output_Get:Skeletal Mesh Component Object Reference)
   TryGetPawnOwner()
  ]}
]
variables [
 Head look Up Alpha {type=number}
 Head Tilt To Right Alpha {type=number}
 Head turn to the left Alpha {type=number}
 Head turn to the right Alpha {type=number}
 Is Skydiving Floating or Dropping {type=boolean}
 L Arm Raise Alpha {type=number}
 Melee Head turn to the left Alpha {type=number}
 Melee Head turn to the right Alpha {type=number}
 Movement Forward or Backward Alpha {type=number}
 Movement to Left Alpha {type=number}
 Movement to Right Alpha {type=number}
 R Arm Raise Alpha {type=number}
 Skeletal Mesh {type=SkeletalMeshComponent}
 Skydiving Movement Forward or Backward Alpha {type=number}
 Skydiving Movement to Left Alpha {type=number}
 Skydiving Movement to Right Alpha {type=number}
 Spine rotation to the Left while Targeting Alpha {type=number}
 Spine rotation to the Right while Targeting Alpha {type=number}
]
