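# Animation Blueprint for the Female Commando 01 hat accessory skeleton, adding physics motion to its left and right straps.
# On initialize it stores the GetOwnerSkeletalMeshComponent result as MeshToCopy, the source for Copy Pose From Mesh.
# The copied pose runs through chained Anim Dynamics nodes on Strap_Left_00 and Strap_Right_00 with angular limits, then converts to local space.
# Output Pose shows no input from that chain or from the separate DefaultSlot branch.
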AnimBlueprint /BRCosmetics/Accessories/Hats/Mesh/Female_Commando_01_Skeleton_AnimBP
target_skeleton=/Game/Accessories/Hats/Mesh/Female_Commando_01_Skeleton
graphs [
 AnimGraph {nodes [
   n1 = Get MeshToCopy() -> (MeshToCopy:Skeletal Mesh Component Object Reference)
   n2 = Anim Dynamics - Bone: Strap_Left_00(ComponentPose=Local To Component(LocalPose=Copy Pose From Mesh(SourceMeshComponent=@n1)), Alpha=1, ExternalForce="0,0,0") {
    bGravityOverrideInSimSpace=true
    alphaBoolBlend {blendOption=Linear}
    boundBone {boneName=Strap_Left_00}
    physicsBodyDefinitions [{
      sphereCollisionRadius=0
      boundBone {boneName=Strap_Left_00}
      boxExtents=<18 18 18>
      constraintSetup {angularLimitsMax=<86.3999 79.2001 36.0001>, angularLimitsMin=<-28.8001 -64.7999 -28.8>, angularTarget=<0 0 0>}
      localJointOffset=<3 0 0>
     }]
   }
   n3 = Anim Dynamics - Bone: Strap_Right_00(ComponentPose=@n2, Alpha=1, ExternalForce="0,0,0") {
    bGravityOverrideInSimSpace=true
    alphaBoolBlend {blendOption=Linear}
    boundBone {boneName=Strap_Right_00}
    physicsBodyDefinitions [{
      sphereCollisionRadius=0
      boundBone {boneName=Strap_Right_00}
      boxExtents=<21 21 21>
      constraintSetup {angularLimitsMax=<86.3999 79.2001 36.0001>, angularLimitsMin=<-28.8001 -64.7999 -28.8>, angularTarget=<0 0 0>}
      localJointOffset=<-3 0 0>
     }]
   }
   Component To Local(ComponentPose=@n3)
   Local To Component(LocalPose=Slot 'DefaultSlot'())
   Output Pose() {layerGroup=None}
  ]}
 EventGraph {nodes [
   Event BlueprintInitializeAnimation()
   > Set MeshToCopy(MeshToCopy:Skeletal Mesh Component Object Reference=GetOwnerSkeletalMeshComponent()) -> (Output_Get:Skeletal Mesh Component Object Reference)
   Event BlueprintUpdateAnimation()
  ]}
]
variables [MeshToCopy {type=SkeletalMeshComponent}]
