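# Physics asset for the Female_Commando_03 hat accessory mesh in BRCosmetics, providing its simple collision shape.
# It holds a single body on joint1, built from one offset and rotated capsule.
# It has no constraints, so no jointed chain or ragdoll motion is defined.
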PhysicsAsset /BRCosmetics/Accessories/Hats/Mesh/Female_Commando_03_PhysicsAsset
bodies {joint1 [{shapeType=Capsule, length=9.9107, radius=8.6178, center=<0.328 -0.3209 10.3111>, rotation=<rot 0 -170.758 -64.0544>}]}
constraints []
