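# Physics asset for the Male Commando 19 hat cosmetic accessory.
# It holds a single capsule body on the Root bone, offset and rotated to fit the mesh.
# There are no constraints, so the hat has one rigid collision shape with no jointed or secondary-motion parts.
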PhysicsAsset /BRCosmetics/Accessories/Hats/Mesh/Male_Commando_19_PhysicsAsset
bodies {Root [{shapeType=Capsule, length=14.5473, radius=11.0995, center=<3.3773 3.5099 0.0008>, rotation=<rot 0 -138.0498 -89.9851>}]}
constraints []
