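# Physics asset for the Male Commando Hat 03 cosmetic accessory in BRCosmetics.
# It holds a single body on Male_Commando_03, shaped as one offset and rotated capsule.
# It has no constraints, so the hat collides as one piece with no jointed parts.
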PhysicsAsset /BRCosmetics/Accessories/Hats/Mesh/Male_Commando_Hat_03_PhysicsAsset
bodies {Male_Commando_03 [{shapeType=Capsule, length=9.8734, radius=8.5336, center=<0.0561 -0.7156 12.8118>, rotation=<rot 0 -175.0927 -64.2247>}]}
constraints []
